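# Sound cue for the drink sound of the Agile Chug Jug consumable.
# A single wave player feeds the AgileChugJug_Use wave straight to the output at boosted volume, with no randomization or mixing.
# It uses the Fortnite consumable-use attenuation and consumables sound class, with a concurrency override that stops the oldest instance once the limit is reached.
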SoundCue /AgileChugJugGameplay/Assets/Audio/Item_Consumables_Agile_Chug_Jug_Drink_Cue
root=WavePlayer_2
nodes [{id=WavePlayer_2, properties {soundWaveAssetPtr=&/AgileChugJug_Use}}]
properties {
 attenuationSettings=&Fort_Consumable_Use_Athena_Att
 bOverrideConcurrency=true
 maxDistance=2250
 soundClassObject=&Fort_Consumables
 volumeMultiplier=1.4
 concurrencyOverrides {maxCount=4, resolutionRule=StopOldest}
}
