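# Sound cue for the hit sound of the Agile Chug Jug consumable item.
# A Random node picks between two equally weighted hit waves, and a Modulator at default settings varies the pitch and volume of the result.
# It plays at reduced volume, borrowing the weapon-drop attenuation and the loot-collect sound class.
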SoundCue /AgileChugJugGameplay/Assets/Audio/Item_Consumables_Agile_Chug_Jug_Hit_Cue
root=Modulator_1
nodes [
 {id=Modulator_1, children [Random_0]}
 {id=Random_0, children [WavePlayer_1,WavePlayer_2], properties {weights [1,1]}}
 {id=WavePlayer_1, properties {soundWaveAssetPtr=&/AgileChugJug_Hit_2}}
 {id=WavePlayer_2, properties {soundWaveAssetPtr=&/AgileChugJug_Hit_1}}
]
properties {
 attenuationSettings=&Fort_Weapon_Drop_Att
 maxDistance=3000
 soundClassObject=&Fort_Loot_Collect
 volumeMultiplier=0.4
}
