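# Sound cue for the Agile Chug Jug consumable in flight, layering a throw sound with an in-air loop.
# A boosted mixer combines a random pick of three throw waves with a looping wave pitched down to a fixed value.
# Each layer has its own modulator and attenuation node.
# The throw layer overrides its attenuation asset with HRTF spatialization, distance low-pass filtering and a distant reverb submix send.
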SoundCue /AgileChugJugGameplay/Assets/Audio/Item_Consumables_Agile_Chug_Jug_InAirLoop_Cue
root=Mixer_0
nodes [
 {
  id=Attenuation_0
  children [Modulator_0]
  properties {
   attenuationSettings=/Game/Sounds/Attenuation/Fort_Weapon_Projectile_InAir_Launch_Throw_Att
   bOverrideAttenuation=true
   attenuationOverrides {
    bAttenuateWithLPF=true
    bEnableSubmixSends=true
    falloffDistance=5000
    hPFFrequencyAtMax=100
    lPFFrequencyAtMax=1500
    lPFRadiusMax=1250
    lPFRadiusMin=0
    reverbDistanceMax=1024
    reverbDistanceMin=0
    reverbWetLevelMax=0
    reverbWetLevelMin=1
    spatializationAlgorithm=SPATIALIZATION_HRTF
    stereoSpread=0
    attenuationShapeExtents {x=0}
    submixSendSettings [{
      bEnableHPFCutoff=false
      bEnableLPFCutoff=false
      disableManualSendClamp=false
      hPFCutoff=20
      lPFCutoff=20000
      maxSendDistance=7500
      maxSendLevel=0.25
      minSendDistance=0
      minSendLevel=0
      sendLevel=0.2
      sendLevelControlMethod=Linear
      soundSubmix=&Fortnite_Reverb_Distant
      customSendLevelCurve {externalCurve=/Game/Sounds/SubMixes/Reverb/Fortnite_Reverb_Distant_Send}
     }]
   }
  }
 }
 {id=Attenuation_1, children [Modulator_1], properties {attenuationSettings=/Game/Sounds/Attenuation/Fort_Weapon_Projectile_InAir_Loop_Att}}
 {id=Mixer_0, children [Attenuation_0, Attenuation_1], properties {inputVolume [1.5,1.5]}}
 {id=Modulator_0, children [Random_0]}
 {id=Modulator_1, children [WavePlayer_4], properties {pitchMax=0.8, pitchMin=0.8}}
 {id=Random_0, children [WavePlayer_1,WavePlayer_2,WavePlayer_3], properties {weights [1,1,1]}}
 {id=WavePlayer_1, properties {soundWaveAssetPtr=&/AgileChugJug_Throw_1}}
 {id=WavePlayer_2, properties {soundWaveAssetPtr=&/AgileChugJug_Throw_2}}
 {id=WavePlayer_3, properties {soundWaveAssetPtr=&/AgileChugJug_Throw_3}}
 {id=WavePlayer_4, properties {bLooping=true, soundWaveAssetPtr=&/AgileChugJug_InAirLoop_1}}
]
properties {maxDistance=5000}
