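# Sound cue for picking up the Agile Chug Jug consumable, reusing the standard Chug Jug pickup wave from Athena content.
# A single wave player feeds a modulator that slightly randomizes pitch while keeping volume constant.
# It plays at reduced volume in the loot-collect sound class with weapon pickup attenuation.
# It references a shared pickup concurrency asset and also carries single-instance, prevent-new concurrency override values.
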SoundCue /AgileChugJugGameplay/Assets/Audio/Item_Consumables_Agile_Chug_Jug_Pickup_Cue
root=Modulator_0
nodes [
 {id=Modulator_0, children [WavePlayer_1], properties {pitchMax=1.15, pitchMin=0.9, volumeMax=1, volumeMin=1}}
 {id=WavePlayer_1, properties {soundWaveAssetPtr=/Game/Athena/Sounds/Items/Chug_Jug/Item_Consumables_Chug_Jug_Pickup_01}}
]
properties {
 attenuationSettings=&Fort_Weapon_Pickup_Att
 maxDistance=3500
 soundClassObject=&Fort_Loot_Collect
 volumeMultiplier=0.6
 concurrencyOverrides {maxCount=1, resolutionRule=PreventNew}
 concurrencySet [&PickupMaximum_Oldest]
}
