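# Sound cue for stopping the Agile Chug Jug consumable, built from two cancel sound waves.
# A random node picks one of the two equally weighted waves and feeds a modulator that applies slight downward pitch variation.
# It plays at reduced volume with weapon-reload attenuation, the loot-collect sound class and a concurrency override that caps simultaneous instances.
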SoundCue /AgileChugJugGameplay/Assets/Audio/Item_Consumables_Agile_Chug_Jug_Stop_Cue
root=Modulator_0
nodes [
 {id=Modulator_0, children [Random_0], properties {pitchMax=1, pitchMin=0.9, volumeMax=1, volumeMin=1}}
 {id=Random_0, children [WavePlayer_4,WavePlayer_5], properties {bShouldExcludeFromBranchCulling=true, weights [1,1]}}
 {id=WavePlayer_4, properties {soundWaveAssetPtr=&/AgileChugJug_Cancel_2}}
 {id=WavePlayer_5, properties {soundWaveAssetPtr=&/AgileChugJug_Cancel_1}}
]
properties {
 attenuationSettings=&Fort_Weapon_Reload_Att
 bOverrideConcurrency=true
 maxDistance=2250
 soundClassObject=&Fort_Loot_Collect
 volumeMultiplier=0.5
 concurrencyOverrides {maxCount=3, volumeScaleAttackTime=0.5}
}
